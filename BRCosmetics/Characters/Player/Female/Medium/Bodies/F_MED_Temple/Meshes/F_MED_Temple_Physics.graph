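# Physics asset for the female medium Temple body cosmetic, covering its dangling arm straps, strands, strings and loops.
# Small capsule bodies on dyn_ bones form one- or two-link chains hung from four parent anchor bodies.
# Swing and twist limits keep strands, strings and loops fairly tight at their roots, while arm straps swing widely.
# An extra constraint links the two right arm strap chains to each other.
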PhysicsAsset /BRCosmetics/Characters/Player/Female/Medium/Bodies/F_MED_Temple/Meshes/F_MED_Temple_Physics
bodies {
 dyn_armStrapA_01_r [{shapeType=Capsule, length=2.75, radius=1.75, center=<1 0 0>, rotation=<rot 90 0 0>}]
 dyn_armStrapA_02_r [{shapeType=Capsule, length=1.25, radius=0.75, center=<1 0 0>, rotation=<rot 90 0 0>}]
 dyn_armStrapB_01_r [{shapeType=Capsule, length=2.75, radius=1.75, center=<1 0 0>, rotation=<rot 90 0 0>}]
 dyn_armStrapB_02_r [{shapeType=Capsule, length=1.75, radius=1, center=<1 0 0>, rotation=<rot 90 0 0>}]
 dyn_loopA_01 [{shapeType=Capsule, length=1.5, radius=2, center=<0.5 0 0>, rotation=<rot 90 0 0>}]
 dyn_loopA_02 [{shapeType=Capsule, length=1.5, radius=1, center=<2 0 0>, rotation=<rot 90 0 0>}]
 dyn_loopB_01 [{shapeType=Capsule, length=1.5, radius=2, center=<0.5 0 0>, rotation=<rot 90 0 0>}]
 dyn_loopB_02 [{shapeType=Capsule, length=1.5, radius=1, center=<2 0 0>, rotation=<rot 90 0 0>}]
 dyn_parent_arm_strap [{shapeType=Capsule, length=0.505, radius=0.505, rotation=<rot 0 -90 0>}]
 dyn_parent_strand_l [{shapeType=Capsule, length=0.505, radius=0.505, rotation=<rot 0 -90 0>}]
 dyn_parent_strand_r [{shapeType=Capsule, length=0.505, radius=0.505, rotation=<rot 0 -90 0>}]
 dyn_parent_strings [{shapeType=Capsule, length=0.505, radius=0.505, rotation=<rot 0 -90 0>}]
 dyn_strandA_01_l [{shapeType=Capsule, length=2, radius=1, center=<3 0 0>, rotation=<rot 90 0 0>}]
 dyn_strandA_01_r [{shapeType=Capsule, length=1.75, radius=1.75, center=<1.2 0 0>, rotation=<rot 90 0 0>}]
 dyn_strandA_02_r [{shapeType=Capsule, length=1.5598, radius=1.0278, center=<1.3 -0.0517 -0.0443>, rotation=<rot 0 94.5983 -88.9412>}]
 dyn_strandB_01_l [{shapeType=Capsule, length=1.25, radius=0.75, center=<3 0 0>, rotation=<rot 90 0 0>}]
 dyn_strandB_01_r [{shapeType=Capsule, length=3, radius=1.5, center=<1 0 0>, rotation=<rot 90 0 0>}]
 dyn_strandB_02_r [{shapeType=Capsule, length=1.8876, radius=0.8197, center=<2.1163 -0.1226 -0.0445>, rotation=<rot 0 81.5253 -86.3911>}]
 dyn_stringA_01 [{shapeType=Capsule, length=2.25, radius=2.75, center=<1 0 0>, rotation=<rot 90 0 0>}]
 dyn_stringA_02 [{shapeType=Capsule, length=1.75, radius=1.25, center=<2 0 0>, rotation=<rot 90 0 0>}]
 dyn_stringB_01 [{shapeType=Capsule, length=3, radius=2, center=<1 0 0>, rotation=<rot 90 0 0>}]
 dyn_stringB_02 [{shapeType=Capsule, length=1.5, radius=1.5, center=<2 0 0>, rotation=<rot 90 0 0>}]
}
constraints [
 {bone1Name=dyn_armStrapA_01_r, bone2Name=dyn_parent_arm_strap, swing1LimitDegrees=120, swing2LimitDegrees=120, twistLimitDegrees=5}
 {bone1Name=dyn_armStrapA_02_r, bone2Name=dyn_armStrapA_01_r, swing1LimitDegrees=30, swing2LimitDegrees=30, twistLimitDegrees=5}
 {bone1Name=dyn_armStrapB_01_r, bone2Name=dyn_parent_arm_strap, swing1LimitDegrees=120, swing2LimitDegrees=120, twistLimitDegrees=5}
 {bone1Name=dyn_armStrapB_02_r, bone2Name=dyn_armStrapB_01_r, swing1LimitDegrees=30, swing2LimitDegrees=30, twistLimitDegrees=5}
 {bone1Name=dyn_strandA_01_l, bone2Name=dyn_parent_strand_l, swing1LimitDegrees=10, swing2LimitDegrees=10, twistLimitDegrees=5}
 {bone1Name=dyn_strandB_01_l, bone2Name=dyn_parent_strand_l, swing1LimitDegrees=10, swing2LimitDegrees=15, twistLimitDegrees=5}
 {bone1Name=dyn_strandB_01_r, bone2Name=dyn_parent_strand_r, swing1LimitDegrees=5, swing2LimitDegrees=5, twistLimitDegrees=5}
 {bone1Name=dyn_strandA_01_r, bone2Name=dyn_parent_strand_r, swing1LimitDegrees=5, swing2LimitDegrees=5, twistLimitDegrees=5}
 {bone1Name=dyn_loopA_01, bone2Name=dyn_parent_strings, swing1LimitDegrees=5, swing2LimitDegrees=5, twistLimitDegrees=1}
 {bone1Name=dyn_loopA_02, bone2Name=dyn_loopA_01, swing1LimitDegrees=5, swing2LimitDegrees=5, twistLimitDegrees=1}
 {bone1Name=dyn_loopB_01, bone2Name=dyn_parent_strings, swing1LimitDegrees=5, swing2LimitDegrees=5, twistLimitDegrees=1}
 {bone1Name=dyn_loopB_02, bone2Name=dyn_loopB_01, swing1LimitDegrees=5, swing2LimitDegrees=5, twistLimitDegrees=1}
 {bone1Name=dyn_stringA_01, bone2Name=dyn_parent_strings, swing1LimitDegrees=8, swing2LimitDegrees=8, twistLimitDegrees=1}
 {bone1Name=dyn_stringA_02, bone2Name=dyn_stringA_01, swing1LimitDegrees=15, swing2LimitDegrees=15, twistLimitDegrees=1}
 {bone1Name=dyn_stringB_01, bone2Name=dyn_parent_strings, swing1LimitDegrees=8, swing2LimitDegrees=8, twistLimitDegrees=1}
 {bone1Name=dyn_stringB_02, bone2Name=dyn_stringB_01, swing1LimitDegrees=15, swing2LimitDegrees=15, twistLimitDegrees=1}
 {bone1Name=dyn_armStrapB_01_r, bone2Name=dyn_armStrapA_01_r, swing1LimitDegrees=45, swing2LimitDegrees=45, twistLimitDegrees=45}
 {bone1Name=dyn_strandA_02_r, bone2Name=dyn_strandA_01_r, swing1LimitDegrees=45, swing2LimitDegrees=45, twistLimitDegrees=45}
 {bone1Name=dyn_strandB_02_r, bone2Name=dyn_strandB_01_r, swing1LimitDegrees=45, swing2LimitDegrees=45, twistLimitDegrees=45}
]
