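# Secondary-motion AnimBlueprint for the F_MED_Temple female cosmetic body, layered on the pose copied from its attached parent mesh.
# Weighted constraints position strap and band bones, then additive glute correctives scale with thigh-to-pelvis rotation computed in the EventGraph.
# AnimDynamics and RigidBody simulate packs, bands, necklace and physics-asset bodies, with pack dynamics reduced when the pawn is stationary.
# A final blended offset repositions the strap lock during skydiving.
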AnimBlueprint /BRCosmetics/Characters/Player/Female/Medium/Bodies/F_MED_Temple/Meshes/F_MED_Temple_AnimBP
target_skeleton=&/F_MED_Temple_Skeleton
graphs [
 AnimGraph {nodes [
   n1 = Get pack_alpha() -> (pack_alpha:double)
   n2 = "Constraint - dyn_parent_arm_strap (PPP)"!(
    ComponentPose=Local To Component(LocalPose=Copy Pose From Mesh() {bUseAttachedParent=true})
    Alpha=1
    ConstraintWeights_0=0.6659
    ConstraintWeights_1:float=0.162
    ConstraintWeights_2:float=0.1721
   ) -> (Pose) {
    alphaBoolBlend {blendOption=Linear}
    boneToModify {boneName=dyn_parent_arm_strap}
    constraintSetup [{targetBone {boneName=upperarm_twist_01_r}}, {targetBone {boneName=upperarm_r}}, {targetBone {boneName=upperarm_twist_02_r}}]
    constraintWeights [1,1,1]
   }
   n3 = "Constraint - dyn_strap_lock (PPPPPPPP)"!(
    ComponentPose=@n2
    Alpha=1
    ConstraintWeights_0=0.0388
    ConstraintWeights_1:float=0.1458
    ConstraintWeights_2:float=0.237
    ConstraintWeights_3:float=0.1041
    ConstraintWeights_4:float=0.2835
    ConstraintWeights_5:float=0.0793
    ConstraintWeights_6:float=0.0202
    ConstraintWeights_7:float=0.0912
   ) -> (Pose) {
    alphaBoolBlend {blendOption=Linear}
    boneToModify {boneName=dyn_strap_lock}
    constraintSetup [
     {targetBone {boneName=clavicle_l}}
     {targetBone {boneName=spine_01}}
     {targetBone {boneName=spine_02}}
     {targetBone {boneName=spine_03}}
     {targetBone {boneName=spine_04}}
     {targetBone {boneName=spine_05}}
     {targetBone {boneName=upperarm_twist_01_l}}
     {targetBone {boneName=pelvis}}
    ]
    constraintWeights [1,1,1,1,1,1,1,1]
   }
   n4 = "Constraint - dyn_parent_band_01 (PPPP)"!(
    ComponentPose=@n3
    Alpha=1
    ConstraintWeights_0=0.1299
    ConstraintWeights_1:float=0.0354
    ConstraintWeights_2:float=0.3868
    ConstraintWeights_3:float=0.448
   ) -> (Pose) {
    alphaBoolBlend {blendOption=Linear}
    boneToModify {boneName=dyn_parent_band_01}
    constraintSetup [{targetBone {boneName=clavicle_r}}, {targetBone {boneName=spine_03}}, {targetBone {boneName=spine_04}}, {targetBone {boneName=spine_05}}]
    constraintWeights [1,1,1,1]
   }
   n5 = "Constraint - dyn_band_01 (PPPP)"!(
    ComponentPose=@n4
    Alpha=1
    ConstraintWeights_0=0.1299
    ConstraintWeights_1:float=0.0354
    ConstraintWeights_2:float=0.3868
    ConstraintWeights_3:float=0.448
   ) -> (Pose) {
    alphaBoolBlend {blendOption=Linear}
    boneToModify {boneName=dyn_band_01}
    constraintSetup [{targetBone {boneName=clavicle_r}}, {targetBone {boneName=spine_03}}, {targetBone {boneName=spine_04}}, {targetBone {boneName=spine_05}}]
    constraintWeights [1,1,1,1]
   }
   n6 = "Constraint - dyn_band_02 (PPPPPPPP)"!(
    ComponentPose=@n5
    Alpha=1
    ConstraintWeights_0=0.0439
    ConstraintWeights_1:float=0.0907
    ConstraintWeights_2:float=0.1845
    ConstraintWeights_3:float=0.1685
    ConstraintWeights_4:float=0.3357
    ConstraintWeights_5:float=0.1021
    ConstraintWeights_6:float=0.0461
    ConstraintWeights_7:float=0.0285
   ) -> (Pose) {
    alphaBoolBlend {blendOption=Linear}
    boneToModify {boneName=dyn_band_02}
    constraintSetup [
     {targetBone {boneName=clavicle_l}}
     {targetBone {boneName=spine_01}}
     {targetBone {boneName=spine_02}}
     {targetBone {boneName=spine_03}}
     {targetBone {boneName=spine_04}}
     {targetBone {boneName=spine_05}}
     {targetBone {boneName=clavicle_r}}
     {targetBone {boneName=pelvis}}
    ]
    constraintWeights [1,1,1,1,1,1,1,1]
   }
   n7 = "Constraint - dyn_band_03 (PPPPP)"!(
    ComponentPose=@n6
    Alpha=1
    ConstraintWeights_0=0.1498
    ConstraintWeights_1:float=0.28
    ConstraintWeights_2:float=0.1385
    ConstraintWeights_3:float=0.3581
    ConstraintWeights_4:float=0.0735
   ) -> (Pose) {
    alphaBoolBlend {blendOption=Linear}
    boneToModify {boneName=dyn_band_03}
    constraintSetup [
     {targetBone {boneName=spine_01}}
     {targetBone {boneName=spine_02}}
     {targetBone {boneName=spine_03}}
     {targetBone {boneName=spine_04}}
     {targetBone {boneName=pelvis}}
    ]
    constraintWeights [1,1,1,1,1]
   }
   n8 = "Transform (Modify) Bone - Bone: dyn_glute_r"(ComponentPose=@n7, Alpha=Get thigh_rot_r(), Translation="3,-2,0", Rotation="0,0,-20", Scale="1.2,1.2,1.2") {
    rotationMode=BMM_Additive
    rotationSpace=BCS_BoneSpace
    scaleMode=BMM_Additive
    scaleSpace=BCS_BoneSpace
    translationMode=BMM_Additive
    translationSpace=BCS_BoneSpace
    alphaBoolBlend {blendOption=Linear}
    boneToModify {boneName=dyn_glute_r}
    rotation {roll=-20}
    scale=<1.2 1.2 1.2>
    translation=<3 -2>
   }
   n9 = "Transform (Modify) Bone - Bone: dyn_glute_l"(ComponentPose=@n8, Alpha=Get thigh_rot_l(), Translation="-3.5,-4,0", Rotation="0,0,-20", Scale="1.35,1.35,1.35") {
    rotationMode=BMM_Additive
    rotationSpace=BCS_BoneSpace
    scaleMode=BMM_Additive
    scaleSpace=BCS_BoneSpace
    translationMode=BMM_Additive
    translationSpace=BCS_BoneSpace
    alphaBoolBlend {blendOption=Linear}
    boneToModify {boneName=dyn_glute_l}
    rotation {roll=-20}
    scale=<1.35 1.35 1.35>
    translation=<-3.5 -4>
   }
   n10 = "Transform (Modify) Bone - Bone: dyn_hip_cor_01_l"(ComponentPose=@n9, Alpha=0, Translation="4.101,-0.000,-0.000", Rotation="-5.739,-6.726,0.000", Scale="1,1,1") {
    rotationMode=BMM_Additive
    rotationSpace=BCS_BoneSpace
    scaleSpace=BCS_BoneSpace
    translationMode=BMM_Additive
    translationSpace=BCS_BoneSpace
    alphaBoolBlend {blendOption=Linear}
    boneToModify {boneName=dyn_hip_cor_01_l}
    rotation=<rot -5.739 -6.7261 0>
    translation {x=4.101}
   }
   n11 = Anim Dynamics - Bone: dyn_packA_01(ComponentPose=@n10, Alpha=@n1, ExternalForce="0,0,0") {
    angularDampingOverride=1
    angularSpringConstant=1500
    bAngularSpring=true
    bGravityOverrideInSimSpace=true
    bOverrideAngularDamping=true
    gravityScale=0.38
    simulationSpace=RootRelative
    alphaBoolBlend {blendOption=Linear}
    boundBone {boneName=dyn_packA_01}
    componentAppliedLinearAccClamp=<1000 1000 1000>
    componentLinearAccScale=<0.2 0.2 0.15>
    componentLinearVelScale=<0.2 0.2 0.15>
    physicsBodyDefinitions [{
      sphereCollisionRadius=0
      boundBone {boneName=dyn_packA_01}
      boxExtents=<20 25 25>
      constraintSetup {angularLimitsMax=<0 0 0>, angularLimitsMin=<0 0 -40>, angularTarget=<0 0 0>}
      localJointOffset=<15 0 0>
     }]
   }
   n12 = Anim Dynamics - Bone: dyn_packB_01(ComponentPose=@n11, Alpha=@n1, ExternalForce="0,0,0") {
    angularDampingOverride=1
    angularSpringConstant=1500
    bAngularSpring=true
    bGravityOverrideInSimSpace=true
    bOverrideAngularDamping=true
    gravityScale=0.35
    simulationSpace=RootRelative
    alphaBoolBlend {blendOption=Linear}
    boundBone {boneName=dyn_packB_01}
    componentAppliedLinearAccClamp=<1000 1000 1100>
    componentLinearAccScale=<0.3 0.3 0.25>
    componentLinearVelScale=<0.3 0.3 0.25>
    physicsBodyDefinitions [{
      sphereCollisionRadius=0
      boundBone {boneName=dyn_packB_01}
      boxExtents=<10 20 20>
      constraintSetup {angularLimitsMax=<0 1.5 0>, angularLimitsMin=<0 -1.5 -40>, angularTarget=<0 0 0>}
      localJointOffset=<10 0 0>
     }]
   }
   n13 = RigidBody(ComponentPose=@n12) {
    bOverrideWorldGravity=true
    overridePhysicsAsset=&/F_MED_Temple_Physics
    simulationSpace=BaseBoneSpace
    alphaBoolBlend {blendOption=Linear}
    baseBoneRef {boneName=root}
    componentAppliedLinearAccClamp=<1000 1000 1000>
    componentLinearAccScale=<0.3 0.3 0.3>
    componentLinearVelScale=<0.3 0.3 0.3>
    overrideWorldGravity {z=-980}
   }
   n14 = Anim Dynamics - Bone: dyn_band_01(ComponentPose=@n13, Alpha=1, ExternalForce="0,0,0") {
    angularDampingOverride=1
    angularSpringConstant=1500
    bAngularSpring=true
    bGravityOverrideInSimSpace=true
    bOverrideAngularDamping=true
    gravityScale=0.35
    simulationSpace=RootRelative
    alphaBoolBlend {blendOption=Linear}
    boundBone {boneName=dyn_band_01}
    componentLinearAccScale=<0.5 0.5 0.5>
    componentLinearVelScale=<0.5 0.5 0.5>
    physicsBodyDefinitions [{
      sphereCollisionRadius=0
      boundBone {boneName=dyn_band_01}
      boxExtents=<10 20 20>
      constraintSetup {angularLimitsMax=<0 0 0>, angularLimitsMin=<0 0 -15>, angularTarget=<0 0 0>}
      localJointOffset=<5 0 0>
     }]
   }
   n15 = Anim Dynamics - Bone: dyn_band_02(ComponentPose=@n14, Alpha=1, ExternalForce="0,0,0") {
    angularDampingOverride=1
    angularSpringConstant=1500
    bAngularSpring=true
    bGravityOverrideInSimSpace=true
    bOverrideAngularDamping=true
    gravityScale=0.35
    simulationSpace=RootRelative
    alphaBoolBlend {blendOption=Linear}
    boundBone {boneName=dyn_band_02}
    componentLinearAccScale=<0.5 0.5 0.5>
    componentLinearVelScale=<0.5 0.5 0.5>
    physicsBodyDefinitions [{
      sphereCollisionRadius=0
      boundBone {boneName=dyn_band_02}
      boxExtents=<-8 15 15>
      constraintSetup {angularLimitsMax=<0 2 0>, angularLimitsMin=<0 -2 -20>, angularTarget=<0 0 0>}
      localJointOffset=<8 0 0>
     }]
   }
   n16 = Anim Dynamics - Bone: dyn_band_03(ComponentPose=@n15, Alpha=1, ExternalForce="0,0,0") {
    angularDampingOverride=1
    angularSpringConstant=1500
    bAngularSpring=true
    bGravityOverrideInSimSpace=true
    bOverrideAngularDamping=true
    gravityScale=0.35
    simulationSpace=RootRelative
    alphaBoolBlend {blendOption=Linear}
    boundBone {boneName=dyn_band_03}
    componentLinearAccScale=<0.5 0.5 0.5>
    componentLinearVelScale=<0.5 0.5 0.5>
    physicsBodyDefinitions [{
      sphereCollisionRadius=0
      boundBone {boneName=dyn_band_03}
      boxExtents=<10 15 15>
      constraintSetup {angularLimitsMax=<0 3 0>, angularLimitsMin=<0 -3 -45>, angularTarget=<0 0 0>}
      localJointOffset=<10 0 0>
     }]
   }
   n17 = Anim Dynamics - Bone: dyn_necklace_l_01(ComponentPose=@n16, Alpha=1, ExternalForce="0,0,0") {
    angularDampingOverride=1
    angularSpringConstant=1500
    bAngularSpring=true
    bGravityOverrideInSimSpace=true
    bOverrideAngularDamping=true
    gravityScale=0.35
    numSolverIterationsPreUpdate=8
    simulationSpace=RootRelative
    alphaBoolBlend {blendOption=Linear}
    boundBone {boneName=dyn_necklace_l_01}
    physicsBodyDefinitions [{
      sphereCollisionRadius=0
      boundBone {boneName=dyn_necklace_l_01}
      boxExtents=<20 10 20>
      constraintSetup {angularLimitsMax=<0 5 0>, angularLimitsMin=<0 -5 -10>, angularTarget=<0 0 0>}
      localJointOffset=<10 0 0>
     }]
   }
   n18 = Anim Dynamics - Bone: dyn_necklace_l_03(ComponentPose=@n17, Alpha=1, ExternalForce="0,0,0") {
    angularDampingOverride=1
    angularSpringConstant=1500
    bAngularSpring=true
    bGravityOverrideInSimSpace=true
    bOverrideAngularDamping=true
    gravityScale=0.35
    numSolverIterationsPreUpdate=8
    simulationSpace=RootRelative
    alphaBoolBlend {blendOption=Linear}
    boundBone {boneName=dyn_necklace_l_03}
    physicsBodyDefinitions [{
      sphereCollisionRadius=0
      boundBone {boneName=dyn_necklace_l_03}
      boxExtents=<10 5 10>
      constraintSetup {angularLimitsMax=<0 5 0>, angularLimitsMin=<0 -5 -10>, angularTarget=<0 0 0>}
      localJointOffset=<5 0 0>
     }]
   }
   n19 = "Transform (Modify) Bone - Bone: dyn_strap_lock"!(ComponentPose=@n18, bAlphaBoolEnabled:bool=Get bIsSkyDiving(), Translation="0,-3,0", Rotation="0,0,0", Scale="1,1,1") -> (Pose) {
    alphaInputType=Bool
    translationMode=BMM_Replace
    translationSpace=BCS_BoneSpace
    alphaBoolBlend {blendInTime=0.25, blendOption=Linear, blendOutTime=0.25}
    boneToModify {boneName=dyn_strap_lock}
    translation {y=-3}
   }
   Output Pose(Result=Component To Local(ComponentPose=@n19)) {layerGroup=None}
  ]}
 EventGraph {nodes [
   Event BlueprintUpdateAnimation()
   > Get!(execute:Exec) -> (then:Exec, else:Exec, Skel_Mesh:Skeletal Mesh Component Object Reference)
   > n1 = Sequence()
   n2 = Get Skel_Mesh()
   > Set thigh_rot_l(thigh_rot_l=GetAlphaFromReferenceTransformDelta(SkeletalMeshComponent=@n2, BoneName=thigh_l, BaseName=pelvis, InputRange=(X=-20,Y=-100), TwistAxis="1,0,0"))
   > Set thigh_rot_r(thigh_rot_r=GetAlphaFromReferenceTransformDelta(SkeletalMeshComponent=@n2, BoneName=thigh_r, BaseName=pelvis, InputRange=(X=-20,Y=-100), TwistAxis="1,0,0"))
   n3 = Branch(execute=@n1.then_1, Condition="float > float"(A=Get PawnSpeed2D(), B=0) {is=title K2Node_CallFunction})
   > Set pack_alpha(pack_alpha:double=1) -> (Output_Get:double)
   Set pack_alpha(execute=@n3.else, pack_alpha=0.2)
   Event BlueprintInitializeAnimation()
   > Set Skel_Mesh(Skel_Mesh=GetPartSkeletalMeshComponent())
  ]}
]
variables [
 pack_alpha {type=number}
 Skel_Mesh {type=SkeletalMeshComponent}
 thigh_rot_l {type=number}
 thigh_rot_r {type=number}
]
