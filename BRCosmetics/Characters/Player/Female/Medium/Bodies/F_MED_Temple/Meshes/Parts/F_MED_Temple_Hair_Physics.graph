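# Physics asset for the hair of the F_MED_Temple female medium player cosmetic, giving its dynamic hair bones capsule bodies for physics simulation.
# Four chains hang from a head capsule: left strand, right strand, long left hair and a short tip, each constrained bone to parent.
# The left strand's root joint has free swing and twist, while the other joints carry narrow angle limits, tightest on the tip chain.
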PhysicsAsset /BRCosmetics/Characters/Player/Female/Medium/Bodies/F_MED_Temple/Meshes/Parts/F_MED_Temple_Hair_Physics
bodies {
 dyn_hair_strand_01_l [{shapeType=Capsule, length=2.25, radius=2.25, center=<0.3 0 0>, rotation=<rot 90 0 0>}]
 dyn_hair_strand_01_r [{shapeType=Capsule, length=3.25, radius=4, center=<0.3 0 0>, rotation=<rot 90 0 0>}]
 dyn_hair_strand_02_l [{shapeType=Capsule, length=3.75, radius=2.75, center=<2 0 0>, rotation=<rot 90 0 0>}]
 dyn_hair_strand_02_r [{shapeType=Capsule, length=3.25, radius=2.75, center=<1.05 0 -0.3183>, rotation=<rot 90 0 0>}]
 dyn_hair_strand_03_l [{shapeType=Capsule, length=3.5, radius=1.25, center=<4 0 0>, rotation=<rot 90 0 0>}]
 dyn_hair_strand_03_r [{shapeType=Capsule, length=3.25, radius=2, center=<3 0 0.65>, rotation=<rot 68.8298 180 180>}]
 dyn_hair_tip_01 [{shapeType=Capsule, length=3.25, radius=2.75, center=<2 0 0>, rotation=<rot 90 0 0>}]
 dyn_hair_tip_02 [{shapeType=Capsule, length=2, radius=2, center=<1.05 0 -0.3183>, rotation=<rot 90 0 0>}]
 dyn_lng_hair_01_l [{shapeType=Capsule, length=3.25, radius=4, center=<0.3 0 0>, rotation=<rot 90 0 0>}]
 dyn_lng_hair_02_l [{shapeType=Capsule, length=3.25, radius=2.75, center=<1.05 0 -0.3183>, rotation=<rot 90 0 0>}]
 dyn_lng_hair_03_l [{shapeType=Capsule, length=3.25, radius=2, center=<3 0 0.65>, rotation=<rot 68.8298 180 180>}]
 head [{shapeType=Capsule, length=7.75, radius=5.5, center=<-0.6194 1.9195 0.0001>, rotation=<rot 90 0 -27.2105>}]
}
constraints [
 {
  bone1Name=dyn_hair_strand_01_l
  bone2Name=head
  swing1LimitDegrees=45
  swing1Motion=Free
  swing2LimitDegrees=45
  swing2Motion=Free
  twistLimitDegrees=45
  twistMotion=Free
 }
 {bone1Name=dyn_hair_strand_02_l, bone2Name=dyn_hair_strand_01_l, swing1LimitDegrees=15, swing2LimitDegrees=20, twistLimitDegrees=5}
 {bone1Name=dyn_hair_strand_03_l, bone2Name=dyn_hair_strand_02_l, swing1LimitDegrees=25, swing2LimitDegrees=15, twistLimitDegrees=5}
 {bone1Name=dyn_lng_hair_01_l, bone2Name=head, swing1LimitDegrees=15, swing2LimitDegrees=13, twistLimitDegrees=5}
 {bone1Name=dyn_lng_hair_02_l, bone2Name=dyn_lng_hair_01_l, swing1LimitDegrees=15, swing2LimitDegrees=13, twistLimitDegrees=5}
 {bone1Name=dyn_lng_hair_03_l, bone2Name=dyn_lng_hair_02_l, swing1LimitDegrees=15, swing2LimitDegrees=13, twistLimitDegrees=5}
 {bone1Name=dyn_hair_tip_01, bone2Name=head, swing1LimitDegrees=5, swing2LimitDegrees=5, twistLimitDegrees=5}
 {bone1Name=dyn_hair_tip_02, bone2Name=dyn_hair_tip_01, swing1LimitDegrees=5, swing2LimitDegrees=5, twistLimitDegrees=5}
 {bone1Name=dyn_hair_strand_01_r, bone2Name=head, swing1LimitDegrees=15, swing2LimitDegrees=13, twistLimitDegrees=5}
 {bone1Name=dyn_hair_strand_02_r, bone2Name=dyn_hair_strand_01_r, swing1LimitDegrees=15, swing2LimitDegrees=13, twistLimitDegrees=5}
 {bone1Name=dyn_hair_strand_03_r, bone2Name=dyn_hair_strand_02_r, swing1LimitDegrees=15, swing2LimitDegrees=13, twistLimitDegrees=5}
]
