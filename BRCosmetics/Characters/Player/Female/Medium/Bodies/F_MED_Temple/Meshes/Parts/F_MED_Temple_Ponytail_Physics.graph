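# Physics asset for the ponytail part of the female medium Temple body, defining simulation bodies for hair dynamics.
# A chain of five capsule bodies on the dyn_pony bones hangs from dyn_parent_pony, which links to head and spine_05 capsules.
# Ponytail joints use limited swing with mostly tight twist, the root joint allowing little swing but wider twist, while the head and spine links rotate freely.
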PhysicsAsset /BRCosmetics/Characters/Player/Female/Medium/Bodies/F_MED_Temple/Meshes/Parts/F_MED_Temple_Ponytail_Physics
bodies {
 dyn_parent_pony [{shapeType=Capsule, length=0.505, radius=0.505, rotation=<rot 0 -90 0>}]
 dyn_pony_01 [{shapeType=Capsule, length=5, radius=3, center=<0.1 0 0>, rotation=<rot 90 0 0>}]
 dyn_pony_02 [{shapeType=Capsule, length=3.25, radius=3.75, center=<3 0 0>, rotation=<rot 90 0 0>}]
 dyn_pony_03 [{shapeType=Capsule, length=3.5, radius=5.25, center=<3 0 0>, rotation=<rot 90 0 0>}]
 dyn_pony_04 [{shapeType=Capsule, length=3.75, radius=4.5, center=<3.5 0 0>, rotation=<rot 90 0 0>}]
 dyn_pony_05 [{shapeType=Capsule, length=4, radius=3, center=<6 0 0>, rotation=<rot 90 0 0>}]
 head [{shapeType=Capsule, length=6.75, radius=7.5, center=<3.5252 0.0924 0>, rotation=<rot -90 0 -15.0002>}]
 spine_05 [{shapeType=Capsule, length=13.505, radius=12.005, center=<4.0747 5.6132 0>, rotation=<rot 0 -90 0>}]
}
constraints [
 {bone1Name=dyn_pony_02, bone2Name=dyn_pony_01, swing1LimitDegrees=30, swing2LimitDegrees=30, twistLimitDegrees=3}
 {bone1Name=dyn_pony_03, bone2Name=dyn_pony_02, swing1LimitDegrees=30, swing2LimitDegrees=30, twistLimitDegrees=3}
 {bone1Name=dyn_pony_04, bone2Name=dyn_pony_03, swing1LimitDegrees=30, swing2LimitDegrees=30, twistLimitDegrees=3}
 {bone1Name=dyn_pony_01, bone2Name=dyn_parent_pony, swing1LimitDegrees=5, swing2LimitDegrees=5, twistLimitDegrees=45}
 {
  bone1Name=head
  bone2Name=spine_05
  swing1LimitDegrees=45
  swing1Motion=Free
  swing2LimitDegrees=45
  swing2Motion=Free
  twistLimitDegrees=45
  twistMotion=Free
 }
 {
  bone1Name=dyn_parent_pony
  bone2Name=head
  swing1LimitDegrees=45
  swing1Motion=Free
  swing2LimitDegrees=45
  swing2Motion=Free
  twistLimitDegrees=45
  twistMotion=Free
 }
 {bone1Name=dyn_pony_05, bone2Name=dyn_pony_04, swing1LimitDegrees=45, swing2LimitDegrees=30, twistLimitDegrees=3}
]
