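# Physics asset for the side hair part of the F_MED_Temple female player body, defining collision and joints for dynamic hair strands.
# Four strand chains, one left and three right, hang from tiny MSTR anchor capsules joined to a head capsule with free angular motion.
# Strand capsules use tight swing and twist limits along each chain.
# Free-rotation cross-constraints tie the two front-right chains together at their first two segments.
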PhysicsAsset /BRCosmetics/Characters/Player/Female/Medium/Bodies/F_MED_Temple/Meshes/Parts/F_MED_Temple_Sides
bodies {
 dyn_hairstrand_01_r [{shapeType=Capsule, length=3.3527, radius=2.5, center=<2.2794 0 -1>, rotation=<rot -90 179.9802 -179.9802>}]
 dyn_hairstrand_02_r [{shapeType=Capsule, length=1.6027, radius=1.75, center=<3 0 0>, rotation=<rot -90 180 -180>}]
 dyn_hairstrand_frA_01_r [{shapeType=Capsule, length=4.3527, radius=3, center=<1 0 3>, rotation=<rot -90 -0.0198 0.0198>}]
 dyn_hairstrand_frA_01_r_MSTR [{shapeType=Capsule, length=0.01, radius=0.01}]
 dyn_hairstrand_frA_02_r [{shapeType=Capsule, length=2.8527, radius=2, center=<1.5 0 2>, rotation=<rot -90 -0.0198 0.0198>}]
 dyn_hairstrand_frA_03_r [{shapeType=Capsule, length=2.3527, radius=2, center=<2 0 2>, rotation=<rot -90 -0.0198 0.0198>}]
 dyn_hairstrand_fr_01_l [{shapeType=Capsule, length=3.3527, radius=2.5, center=<2.2794 0 1>, rotation=<rot -90 179.9802 -179.9802>}]
 dyn_hairstrand_fr_01_l_MSTR [{shapeType=Capsule, length=0.01, radius=0.01}]
 dyn_hairstrand_fr_01_r [{shapeType=Capsule, length=4.3527, radius=3, center=<2 0 3>, rotation=<rot -90 -0.0198 0.0198>}]
 dyn_hairstrand_fr_01_r_MSTR [{shapeType=Capsule, length=0.01, radius=0.01}]
 dyn_hairstrand_fr_02_r [{shapeType=Capsule, length=2.8527, radius=2, center=<3 0 2>, rotation=<rot -90 -0.0198 0.0198>}]
 dyn_hairstrand_r_MSTR [{shapeType=Capsule, length=0.01, radius=0.01}]
 head [{shapeType=Capsule, length=9.7659, radius=6.8573, center=<2.2994 1.4172 0>, rotation=<rot -90 0 -11.9269>}]
}
constraints [
 {
  bone1Name=dyn_hairstrand_fr_01_l_MSTR
  bone2Name=head
  swing1LimitDegrees=45
  swing1Motion=Free
  swing2LimitDegrees=45
  swing2Motion=Free
  twistLimitDegrees=45
  twistMotion=Free
 }
 {
  bone1Name=dyn_hairstrand_fr_01_r_MSTR
  bone2Name=head
  swing1LimitDegrees=45
  swing1Motion=Free
  swing2LimitDegrees=45
  swing2Motion=Free
  twistLimitDegrees=45
  twistMotion=Free
 }
 {bone1Name=dyn_hairstrand_fr_01_l, bone2Name=dyn_hairstrand_fr_01_l_MSTR, swing1LimitDegrees=10, swing2LimitDegrees=10, twistLimitDegrees=5}
 {bone1Name=dyn_hairstrand_fr_01_r, bone2Name=dyn_hairstrand_fr_01_r_MSTR, swing1LimitDegrees=25, swing2LimitDegrees=30, twistLimitDegrees=5}
 {bone1Name=dyn_hairstrand_fr_02_r, bone2Name=dyn_hairstrand_fr_01_r, swing1LimitDegrees=30, swing2LimitDegrees=30, twistLimitDegrees=5}
 {
  bone1Name=dyn_hairstrand_frA_01_r_MSTR
  bone2Name=head
  swing1LimitDegrees=45
  swing1Motion=Free
  swing2LimitDegrees=45
  swing2Motion=Free
  twistLimitDegrees=45
  twistMotion=Free
 }
 {bone1Name=dyn_hairstrand_frA_01_r, bone2Name=dyn_hairstrand_frA_01_r_MSTR, swing1LimitDegrees=15, swing2LimitDegrees=20, twistLimitDegrees=5}
 {bone1Name=dyn_hairstrand_frA_02_r, bone2Name=dyn_hairstrand_frA_01_r, swing1LimitDegrees=15, swing2LimitDegrees=15, twistLimitDegrees=5}
 {bone1Name=dyn_hairstrand_frA_03_r, bone2Name=dyn_hairstrand_frA_02_r, swing1LimitDegrees=25, swing2LimitDegrees=25, twistLimitDegrees=5}
 {
  bone1Name=dyn_hairstrand_frA_01_r
  bone2Name=dyn_hairstrand_fr_01_r
  swing1LimitDegrees=45
  swing1Motion=Free
  swing2LimitDegrees=45
  swing2Motion=Free
  twistLimitDegrees=45
  twistMotion=Free
 }
 {
  bone1Name=dyn_hairstrand_frA_02_r
  bone2Name=dyn_hairstrand_fr_02_r
  swing1LimitDegrees=45
  swing1Motion=Free
  swing2LimitDegrees=45
  swing2Motion=Free
  twistLimitDegrees=45
  twistMotion=Free
 }
 {
  bone1Name=dyn_hairstrand_r_MSTR
  bone2Name=head
  swing1LimitDegrees=45
  swing1Motion=Free
  swing2LimitDegrees=45
  swing2Motion=Free
  twistLimitDegrees=45
  twistMotion=Free
 }
 {bone1Name=dyn_hairstrand_01_r, bone2Name=dyn_hairstrand_r_MSTR, swing1LimitDegrees=5, swing2LimitDegrees=5, twistLimitDegrees=5}
 {bone1Name=dyn_hairstrand_02_r, bone2Name=dyn_hairstrand_01_r, swing1LimitDegrees=8, swing2LimitDegrees=8, twistLimitDegrees=5}
]
