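# Post-process style animation blueprint for the F_MED_Temple face accessory, copying the attached parent's pose and adding secondary motion to hair and ponytail.
# Three chained RigidBody nodes simulate sides, hair and ponytail, with a trail controller as the alternative when rigid bodies are disabled.
# When a backpack is worn, weighted constraints and AnimDynamics reshape the ponytail chain.
# The event graph updates state-based gravity for the ponytail, detects backpacks, and sets backpack-type alphas the AnimGraph leaves unused.
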AnimBlueprint /BRCosmetics/Characters/Player/Female/Medium/Bodies/F_MED_Temple/Meshes/Parts/F_MED_Temple_FaceAcc_AnimBP
target_skeleton=&/F_MED_Temple_FaceAcc_Skeleton
graphs [
 AnimGraph {nodes [
   n1 = "Constraint - dyn_pony_01 (PP)"!(
    ComponentPose=Local To Component(LocalPose=Use cached pose 'BackPack_Cache'())
    Alpha=1
    ConstraintWeights_0=0.9
    ConstraintWeights_1:float=0.1
   ) -> (Pose) {
    alphaBoolBlend {blendOption=Linear}
    boneToModify {boneName=dyn_pony_01}
    constraintSetup [{offsetOption=None, targetBone {boneName=head_con_pony_01}}, {targetBone {boneName=neck_01}}]
    constraintWeights [1,1]
   }
   n2 = "Constraint - dyn_pony_02 (PP)"!(ComponentPose=@n1, Alpha=1, ConstraintWeights_0=0.8, ConstraintWeights_1:float=0.2) -> (Pose) {
    alphaBoolBlend {blendOption=Linear}
    boneToModify {boneName=dyn_pony_02}
    constraintSetup [{offsetOption=None, targetBone {boneName=head_con_pony_02}}, {targetBone {boneName=neck_01}}]
    constraintWeights [1,1]
   }
   n3 = "Constraint - dyn_pony_03 (PP)"!(ComponentPose=@n2, Alpha=1, ConstraintWeights_0=0.7, ConstraintWeights_1:float=0.3) -> (Pose) {
    alphaBoolBlend {blendOption=Linear}
    boneToModify {boneName=dyn_pony_03}
    constraintSetup [{offsetOption=None, targetBone {boneName=neck_con_pony_03}}, {targetBone {boneName=head}}]
    constraintWeights [1,1]
   }
   n4 = "Constraint - dyn_pony_04 (PPP)"!(ComponentPose=@n3, Alpha=1, ConstraintWeights_0=0.85, ConstraintWeights_1:float=0.15, ConstraintWeights_2:float=0) -> (Pose) {
    alphaBoolBlend {blendOption=Linear}
    boneToModify {boneName=dyn_pony_04}
    constraintSetup [{offsetOption=None, targetBone {boneName=clav_con_pony_04}}, {targetBone {boneName=head}}, {targetBone {boneName=spine_05}}]
    constraintWeights [1,1,1]
   }
   n5 = "Constraint - dyn_pony_05 (PP)"!(ComponentPose=@n4, Alpha=1, ConstraintWeights_0=1, ConstraintWeights_1:float=0) -> (Pose) {
    alphaBoolBlend {blendOption=Linear}
    boneToModify {boneName=dyn_pony_05}
    constraintSetup [{offsetOption=None, targetBone {boneName=clav_con_pony_05}}, {targetBone {boneName=spine_05}}]
    constraintWeights [1,1]
   }
   n6 = Anim Dynamics - Bone: dyn_pony_04(ComponentPose=@n5, Alpha=1, ExternalForce="0,0,0") {
    angularDampingOverride=1
    angularSpringConstant=1500
    bAngularSpring=true
    bGravityOverrideInSimSpace=true
    bOverrideAngularDamping=true
    gravityScale=0.35
    numSolverIterationsPreUpdate=8
    simulationSpace=RootRelative
    alphaBoolBlend {blendOption=Linear}
    boundBone {boneName=dyn_pony_04}
    physicsBodyDefinitions [{
      sphereCollisionRadius=0
      boundBone {boneName=dyn_pony_04}
      boxExtents=<8 15 15>
      constraintSetup {angularLimitsMax=<0 8 25>, angularLimitsMin=<0 -8 0>, angularTarget=<0 0 0>}
      localJointOffset=<5 0 0>
     }]
   }
   n7 = Anim Dynamics - Bone: dyn_pony_05(ComponentPose=@n6, Alpha=1, ExternalForce="0,0,0") {
    angularDampingOverride=1
    angularSpringConstant=1500
    bAngularSpring=true
    bGravityOverrideInSimSpace=true
    bOverrideAngularDamping=true
    gravityScale=0.35
    numSolverIterationsPreUpdate=8
    simulationSpace=RootRelative
    alphaBoolBlend {blendOption=Linear}
    boundBone {boneName=dyn_pony_05}
    physicsBodyDefinitions [{
      sphereCollisionRadius=0
      boundBone {boneName=dyn_pony_05}
      boxExtents=<10 15 15>
      constraintSetup {angularLimitsMax=<0 10 40>, angularLimitsMin=<0 -10 -3>, angularTarget=<0 0 0>}
      localJointOffset=<5 0 0>
     }]
   }
   Output Pose(Result=Blend Poses by bool(
     BlendPose_0=Component To Local(ComponentPose=@n7)
     BlendPose_1=Use cached pose 'BackPack_Cache'()
     bActiveValue=Get isWearingBackpack() -> (isWearingBackpack:bool)
    ) {blendType=Linear}) {layerGroup=None}
   Save cached pose 'AnimCache'(Pose=Copy Pose From Mesh() {bUseAttachedParent=true}) {cacheName=AnimCache}
   n8 = Trail controller - Bone: dyn_pony_05!(ComponentPose=Local To Component(LocalPose=Use cached pose 'AnimCache'()), Alpha=1, RelaxationSpeedScale:float=1) -> (Pose) {
    bActorSpaceFakeVel=true
    bEnableDebug=true
    bShowLimit=false
    bUsePlanarLimit=true
    chainLength=4
    maxDeltaTime=0.125
    relaxationSpeedScale=1.2
    baseJoint {boneName=root}
    fakeVelocity {z=70}
    planarLimits [{drivingBone {boneName=spine_05}, planeTransform {location=<0 -2 0>, rotation=<rot 0 0 -90>, scale=<1 1 1>}}]
    rotationLimits [
     {limitMax=<0 5 10>, limitMin=<0 -15 -10>}
     {limitMax=<0 10 15>, limitMin=<0 -20 -15>}
     {limitMax=<0 15 15>, limitMin=<0 -25 -15>}
     {limitMax=<0 5 30>, limitMin=<0 -60 -30>}
    ]
    rotationOffsets [<0 0 0>, <0 0 0>, <0 0 0>, <0 0 0>]
    trailBone {boneName=dyn_pony_05}
    trailRelaxationSpeed {editorCurveData {keys [<at 0 30 RCIM_Linear>,<at 1 11 RCIM_Linear>]}}
   }
   n9 = RigidBody(ComponentPose=Local To Component(LocalPose=Use cached pose 'AnimCache'())) {
    bOverrideWorldGravity=true
    overridePhysicsAsset=&/F_MED_Temple_Sides
    simulationSpace=BaseBoneSpace
    alphaBoolBlend {blendOption=Linear}
    baseBoneRef {boneName=root}
    overrideWorldGravity {z=-980}
   }
   n10 = RigidBody(ComponentPose=@n9, Alpha=0.85) {
    bClampLinearTranslationLimitToRefPose=true
    bOverrideWorldGravity=true
    overridePhysicsAsset=&/F_MED_Temple_Hair_Physics
    simulationSpace=BaseBoneSpace
    alphaBoolBlend {blendOption=Linear}
    baseBoneRef {boneName=root}
    componentAppliedLinearAccClamp=<1100 1100 1100>
    componentLinearAccScale=<0.5 0.5 0.5>
    componentLinearVelScale=<0.5 0.5 0.5>
    overrideWorldGravity {z=-980}
   }
   n11 = RigidBody!(ComponentPose=@n10, Alpha, OverrideWorldGravity:Vector=Get gravityVector() -> (gravityVector:Vector), ExternalForce) -> (Pose) {
    bOverrideWorldGravity=true
    overridePhysicsAsset=&/F_MED_Temple_Ponytail_Physics
    simulationSpace=BaseBoneSpace
    alphaBoolBlend {blendOption=Linear}
    baseBoneRef {boneName=root}
    componentAppliedLinearAccClamp=<1000 1000 800>
    componentLinearAccScale=<2 2 2>
    componentLinearVelScale=<2 2 2>
    overrideWorldGravity {z=-980}
   }
   Save cached pose 'BackPack_Cache'(Pose=Blend Poses by bool(BlendPose_0=Component To Local(ComponentPose=@n11), BlendPose_1=Component To Local(ComponentPose=@n8), bActiveValue=Get bRigidBodyEnabled()) {blendType=Linear}) {cacheName=BackPack_Cache}
  ]}
 EventGraph {nodes [
   Event BlueprintUpdateAnimation()
   > n1 = Sequence!(execute) -> (then_0, then_1, then_2:Exec, then_3:Exec)
   > n2 = Gravity Override State Function DEPRECATED(CharacterPartAnimInstance=Self-Reference(), GameTimeInSeconds=GetGameTimeInSeconds(), GravityOverrideParams=Get Gravity Override Params())
   > n3 = Set gravityVector(gravityVector:Vector=@n2) -> (Output_Get:Vector)
   n4 = Branch(execute=@n1.then_2, Condition=Get bRigidBodyEnabled())
   Set templePackAlpha(execute=@n4.else, templePackAlpha:double=0) -> (Output_Get:double)
   > Set isGenericBackpackAlpha(isGenericBackpackAlpha:double=0) -> (Output_Get:double)
   n5 = CheckForBackbling()
   Set isWearingBackpack(execute=@n1.then_1, isWearingBackpack:bool=@n5) -> (Output_Get:bool)
   > Set isWearingCapeBool(isWearingCapeBool:bool=@n5.bIsBackblingLong) -> (Output_Get:bool)
   > n6 = Cast To FortPlayerPawn(Object=TryGetPawnOwner())
   > n7 = Sequence!(execute) -> (then_0, then_1)
   n8 = MatchesWildcard(
    SourceString=GetDisplayName(Object=GetSkeletalMeshForPartType(self=@n6, PartType=Backpack))
    Wildcard=*F_MED_Temple_Pack*
    SearchCase=CaseSensitive
   )
   n9 = NOT Boolean(A=Get isWearingCapeBool() -> (isWearingCapeBool:bool))
   > n10 = Set isTemplePack(isTemplePack:bool=AND Boolean(A=@n8, B=@n9)) -> (Output_Get:bool)
   > Set templePackAlpha(templePackAlpha="To Float (Boolean)"(InBool=@n10))
   n11 = NOT Boolean(A=@n8)
   n12 = Get isWearingBackpack()
   Set isGenericBackpackAlpha(execute=@n7.then_1, isGenericBackpackAlpha="To Float (Boolean)"(InBool=AND Boolean!(A=@n9, B=@n11, C:bool=@n12) -> (ReturnValue)))
   DrawVectorFromJointFunction!(
    execute
    SkelMeshComponent:Scene Component Object Reference
    JointName:Name=head
    LineEnd:Vector=@n3
    JointOffset:Vector="0, 0, 0"
    LineColor:Linear Color Structure=(R=1,G=0,B=0,A=1)
    Thickness:double=2
    Foreground:EDrawDebugSceneDepthPriorityGroup Enum=Foreground
   ) -> (then)
   n13 = CheckForBackbling()
   AND Boolean!(A=@n13, B=NOT Boolean(A=@n13.bIsBackblingLong)) -> (ReturnValue)
   Reroute Node(InputPin:Wildcard) -> (OutputPin:Wildcard)
   Set isWearingCape(isWearingCape:double="To Float (Boolean)"(InBool=@n13.bIsBackblingLong)) -> (Output_Get:double)
  ]}
]
variables [
 Gravity Override Params {
  type=GravityOverrideParamsStruct
  default {skydive_FloatingDown {gravityOverride=<0 -700 2000>}, skydive_GoingUp {gravityOverride=<0 -300 -100>}}
 }
 gravityVector {type=Vector}
 isGenericBackpackAlpha {type=number}
 isTemplePack {type=boolean}
 isWearingBackbling {type=boolean}
 isWearingBackpack {type=boolean}
 isWearingCape {type=number}
 isWearingCapeBool {type=boolean}
 templePackAlpha {type=number}
]
